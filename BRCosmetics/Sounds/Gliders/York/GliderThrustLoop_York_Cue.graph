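# Sound cue for the York glider's in-air thrust loop, spatialized with the glider thrust attenuation and routed through a glider sound class.
# A mixer layers two looping waves, one without siren and one with siren, each through its own continuous modulator.
# The AnyDirection parameter modulates both layers' volume and bends the pitch of the siren-free layer, while the siren's pitch range appears to stay flat.
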SoundCue /BRCosmetics/Sounds/Gliders/York/GliderThrustLoop_York_Cue
root=Mixer_0
nodes [
 {id=Mixer_0, children [ModulatorContinuous_0, ModulatorContinuous_3], properties {inputVolume [1,1]}}
 {
  id=ModulatorContinuous_0
  children [WavePlayer_0]
  properties {
   pitchModulationParams {default=0, maxOutput=1.25, minOutput=0.98, parameterName=AnyDirection}
   volumeModulationParams {default=0, minOutput=0.7, parameterName=AnyDirection}
  }
 }
 {
  id=ModulatorContinuous_3
  children [WavePlayer_3]
  properties {
   pitchModulationParams {default=0, minOutput=1, parameterName=AnyDirection}
   volumeModulationParams {default=0, minOutput=0.7, parameterName=AnyDirection}
  }
 }
 {id=WavePlayer_0, properties {bLooping=true, soundWaveAssetPtr=&/Glider_York_InAir_Loop_No_Siren}}
 {id=WavePlayer_3, properties {bLooping=true, soundWaveAssetPtr=&/Glider_York_InAir_Loop_Siren}}
]
properties {
 attenuationSettings=&Fort_Player_Glider_Thrust_Athena_Att
 audioPropertiesSheet=&APS_Cosmetic_Glider_Loop
 maxDistance=12500
 soundClassObject=&Fort_Gliders_Land
 volumeMultiplier=0.4
}
